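# Masked, two-sided unlit dissolve material for Niagara mesh particles showing a branded mountain graphic.
# A rotated, scrolling mountain texture adds shimmer outside the outline texture, driven and faded by dynamic parameters.
# The Mask parameter shifts a rotated UV gradient through step thresholds, forming a visibility window with thin edge bands tinted by particle color.
# Emissive passes through a shared adjustment function, while time-based sine and cosine build a wiggle vector with particle-controlled Z.
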Material /DealershipSharedLibrary/Assets/ThirdPartyBranding/Mountain/M_MountainDissolve
graph {
 properties {bUsedWithNiagaraMeshParticles=true, blendMode=BLEND_Masked, shadingModel=MSM_Unlit, twoSided=true, usedShadingModels=Unlit}
 nodes [
  n1 = ScalarParameter() {defaultValue=0.7, parameterName=Scale X}
  n2 = ScalarParameter() {defaultValue=0.7, parameterName=Scale Y}
  n3 = DynamicParameter() {defaultValue=<rgba 1 1 0 1>, paramNames [ShimmerAnim,ShimmerO,Mask,WiggleZ]}
  n4 = Add(
   A=Multiply(A=TextureCoordinate(), B=AppendVector(A=@n1, B=@n2))
   B=AppendVector(A=@n3.ShimmerAnim, B=ScalarParameter() {defaultValue=0.15, parameterName=Shimmer})
  )
  n5 = ScalarParameter() {defaultValue=0.34, parameterName=Rotate Texture}
  n6 = Multiply(
   A=TextureSample(UVs=MaterialFunctionCall(UVs=@n4, "Rotation Angle (0-1)"=@n5) {materialFunction=&CustomRotator}.Rotated Values) {texture=&/T_MountainRGB}.R
   B=@n3.ShimmerO
  )
  n7 = TextureSample() {texture=&/T_MountainOutline}
  n8 = OneMinus(None=@n7.RGB)
  n9 = MaterialFunctionCall(UVs=TextureCoordinate(), "Rotation Angle (0-1)"=Constant() {r=0.08}) {materialFunction=&CustomRotator}
  n10 = Power(Base=ComponentMask(None=Add(A=@n9.Rotated Values, B=@n3.Mask) {constB=-0.5}) {g=true}, Exp=Constant() {r=2})
  n11 = Step(X=@n10) {constX=0.11, constY=0.05}
  n12 = Step(X=@n10) {constX=0.11, constY=0.052}
  n13 = Step(X=@n10) {constY=0.89}
  n14 = Step(X=@n10) {constY=0.9}
  n15 = Multiply(A=Multiply(A=ParticleColor().RGB, B=Constant() {r=2}), B=Add(A=Subtract(A=@n11, B=@n12), B=Subtract(A=@n13, B=@n14)))
  MaterialFunctionCall(Emissive=Add(A=Multiply(A=@n6, B=@n8), B=Add(A=@n7.RGB, B=@n15))) {materialFunction=&MF_FX_EmissiveAdjustments}
  n16 = Multiply(A=Time(), B=Constant() {r=1})
  n17 = Sine(None=@n16)
  n18 = Cosine(None=@n16)
  Reroute(None=AppendVector(A=AppendVector(A=@n17, B=@n18), B=Multiply(A=@n17, B=@n3.WiggleZ)))
  Saturate(None=Multiply(A=OneMinus(None=@n14), B=@n11))
 ]
}
